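; Winterfest quest actor for a snow creature prop, shown as either an Ollie or a Fishstick snowman.
; The construction script reads UseVariant to pick Ollie when true and Fishstick otherwise, assigning that mesh to both the pre-interaction and post-interaction components.
; The BeginPlay, overlap and Tick events hold no active logic in this graph.
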(event EventBeginPlay
  (|CallParentFunction) ; This node is disabled and will not be called.\nDrag off pins to build functionality.
)

(event Collision|EventActorBeginOverlap
  (params (OtherActor :type Actor%20Object%20Reference)))

(event EventTick
  (params (DeltaSeconds :type Float%20%28single-precision%29)))

(fn ConstructionScript
  (|CallParentFunction)
  (if (Variables|Default|GetUseVariant)
    (Components|StaticMesh|SetStaticMesh (Variables|Components|GetSM_Quest_PreInteraction) "/Game/Environments/Helios/Props/Quest/Quest_SnowmanOllie_A/Meshes/SM_Quest_SnowmanOllie_A.SM_Quest_SnowmanOllie_A")
    (Components|StaticMesh|SetStaticMesh (Variables|Components|GetSM_Quest_PostInteraction) "/Game/Environments/Helios/Props/Quest/Quest_SnowmanOllie_A/Meshes/SM_Quest_SnowmanOllie_A.SM_Quest_SnowmanOllie_A")
    (else
      (Components|StaticMesh|SetStaticMesh (Variables|Components|GetSM_Quest_PreInteraction) "/Game/Environments/Helios/Props/Quest/Quest_SnowmanFishstick_A/Meshes/SM_Quest_SnowmanFishstick_A.SM_Quest_SnowmanFishstick_A")
      (Components|StaticMesh|SetStaticMesh (Variables|Components|GetSM_Quest_PostInteraction) "/Game/Environments/Helios/Props/Quest/Quest_SnowmanFishstick_A/Meshes/SM_Quest_SnowmanFishstick_A.SM_Quest_SnowmanFishstick_A"))))
